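# MF_POMDepth is a material function that scales Min and Max values by camera depth, apparently step counts for parallax occlusion mapping in the WinterArt materials.
# DepthFromWorldPosition feeds a saturated linear ramp, starting at a fixed near offset and spanning a fixed range.
# That shared alpha blends both scalar inputs toward small constants, so distant pixels receive low Min and Max outputs regardless of the supplied values.
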MaterialFunction /Figment_S10/Athena/WinterArt/Materials/MF_POMDepth
graph {nodes [
  n1 = MaterialFunctionCall!(World Position) {materialFunction=/Engine/Functions/Engine_MaterialFunctions02/Utility/DepthFromWorldPosition}
  n2 = Constant() {r=1303.7531}
  n3 = Saturate(None=Divide(A=Subtract(A=@n1.Result, B=@n2), B=Constant() {r=5000}))
  FunctionOutput(None=LinearInterpolate(A=FunctionInput() {inputName=Max, inputType=FunctionInput_Scalar, sortPriority=1}, B=Constant() {r=2}, Alpha=@n3)) {outputName=Max, sortPriority=1}
  FunctionOutput(None=LinearInterpolate(A=FunctionInput() {inputName=Min, inputType=FunctionInput_Scalar}, B=Constant() {r=1}, Alpha=@n3)) {outputName=Min}
 ]}
